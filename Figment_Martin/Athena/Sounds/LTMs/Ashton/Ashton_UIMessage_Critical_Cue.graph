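# Sound cue for a critical UI message in the Ashton limited-time mode, going by its name and path.
# A single wave player feeds the output directly with an orchestral combat-music sting, the medium-threat ending from the Fort music library.
# It plays at half volume through the important in-game mission sound class, with a very large max distance.
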SoundCue /Figment_Martin/Athena/Sounds/LTMs/Ashton/Ashton_UIMessage_Critical_Cue
root=WavePlayer_5
nodes [{id=WavePlayer_5, properties {soundWaveAssetPtr=/Game/Sounds/Fort_Music/Orchestral_Score/Orch_Combat/Combat_Med_Threat_End_01}}]
properties {maxDistance=2097152, soundClassObject=&Fort_Mission_Ingame_Important_PSM, volumeMultiplier=0.5}
