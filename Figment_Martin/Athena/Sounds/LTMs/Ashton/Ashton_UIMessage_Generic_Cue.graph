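# Sound cue for a generic UI message notification in the Ashton limited-time mode.
# It plays the shared mission-update wave from the Fort UI sounds through a modulator that pins pitch at normal.
# Output is at reduced volume and routed to the important in-game mission sound class.
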SoundCue /Figment_Martin/Athena/Sounds/LTMs/Ashton/Ashton_UIMessage_Generic_Cue
root=Modulator_0
nodes [
 {id=Modulator_0, children [WavePlayer_0], properties {pitchMax=1, pitchMin=1}}
 {id=WavePlayer_0, properties {soundWaveAssetPtr=/Game/Sounds/Fort_UI_Sounds/Gameplay/mission_update_mid}}
]
properties {maxDistance=2097152, soundClassObject=&Fort_Mission_Ingame_Important_PSM, volumeMultiplier=0.5}
