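# Sound cue for the Carmine finger snap, named as the final cue of the Ashton LTM set.
# A single wave player plays the finger snap wave directly, with no randomization or mixing.
# It uses a distant Carmine VO attenuation asset with a very large max distance and an important in-game mission sound class.
# A concurrency override blocks new instances rather than cutting off playing ones.
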SoundCue /Figment_Martin/Athena/Sounds/LTMs/Ashton/SC_Ashton_Carmine_Final_Cue
root=WavePlayer_0
nodes [{id=WavePlayer_0, properties {soundWaveAssetPtr=/Figment_Martin/Sounds/Fort_GamePlay_Sounds/Carmine/Carmine_fingersnap_A}}]
properties {
 attenuationSettings=/Figment_Martin/Sounds/Fort_GamePlay_Sounds/Meteor/Fort_Carmine_VO_Distant_Athena_Att
 bOverrideConcurrency=true
 maxDistance=2097152
 soundClassObject=&Fort_Mission_Ingame_Important_PSM
 volumeMultiplier=1
 concurrencyOverrides {resolutionRule=PreventNew}
}
