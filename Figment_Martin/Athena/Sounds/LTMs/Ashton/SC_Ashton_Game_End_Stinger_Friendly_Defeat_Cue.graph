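# Sound cue for the Ashton LTM end-of-game stinger, played on a friendly defeat.
# A single wave player feeds the Ashton_Stinger_FriendlyDefeat wave straight to the output, with no mixing or randomization.
# It routes through the important in-game mission sound class and sets a very large max distance.
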SoundCue /Figment_Martin/Athena/Sounds/LTMs/Ashton/SC_Ashton_Game_End_Stinger_Friendly_Defeat_Cue
root=WavePlayer_2
nodes [{id=WavePlayer_2, properties {soundWaveAssetPtr=&/Ashton_Stinger_FriendlyDefeat}}]
properties {maxDistance=2097152, soundClassObject=&Fort_Mission_Ingame_Important_PSM, volumeMultiplier=1}
