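# Sound cue for an important UI message in the Ashton limited-time mode.
# A single wave player plays a reused Score LTM final-countdown warning sound.
# It plays at half volume through the important in-game mission sound class, with a very large max distance.
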SoundCue /Figment_Martin/Athena/Sounds/LTMs/Ashton/SC_Ashton_UIMessage_Important_Cue
root=WavePlayer_1
nodes [{id=WavePlayer_1, properties {soundWaveAssetPtr=/Game/Athena/Sounds/LTMs/Score/ScoreLTM_InGamePop_FinalCountdown_Warning_07}}]
properties {maxDistance=2097152, soundClassObject=&Fort_Mission_Ingame_Important_PSM, volumeMultiplier=0.5}
